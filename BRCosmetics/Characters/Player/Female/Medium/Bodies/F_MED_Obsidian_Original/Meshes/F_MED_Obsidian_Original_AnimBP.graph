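# Character-part AnimBP for the female Obsidian body, adding secondary motion to its cape on the copied parent pose.
# Deform Control Rigs, belt/cape constraints and sim-space, slope and crouch bone adjustments build Start_Cache.
# It feeds rigid-body simulation with backbling variants, or Trail Controller fallbacks for normal and DBNO states when rigid body is off.
# EventGraph updates gravity, sim-space rotation, slope, and trail rotation/alpha from movement, swimming and skydiving.
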AnimBlueprint /BRCosmetics/Characters/Player/Female/Medium/Bodies/F_MED_Obsidian_Original/Meshes/F_MED_Obsidian_Original_AnimBP
target_skeleton=&/F_MED_Obsidian_Original_Skeleton
graphs [
 AnimGraph {nodes [
   n1 = "Constraint - dyn_cape_pin_bk_out_l_01_MSTR (PPP)"!(ComponentPose, Alpha=1, ConstraintWeights_0=0.016, ConstraintWeights_1:float=0.6173, ConstraintWeights_2:float=0.3667) -> (Pose) {
    alphaBoolBlend {blendOption=Linear}
    boneToModify {boneName=dyn_cape_pin_bk_out_l_01_MSTR}
    constraintSetup [{targetBone {boneName=deform_pectoral_fr_l}}, {targetBone {boneName=upperarm_twist_01_l}}, {targetBone {boneName=clavicle_l}}]
    constraintWeights [1,1,1]
   }
   "Constraint - dyn_cape_pin_bk_out_r_01_MSTR (PPP)"!(ComponentPose=@n1, Alpha=1, ConstraintWeights_0=0.3667, ConstraintWeights_1:float=0.6173, ConstraintWeights_2:float=0.016) -> (Pose) {
    alphaBoolBlend {blendOption=Linear}
    boneToModify {boneName=dyn_cape_pin_bk_out_r_01_MSTR}
    constraintSetup [{targetBone {boneName=clavicle_r}}, {targetBone {boneName=upperarm_twist_01_r}}, {targetBone {boneName=deform_pectoral_fr_r}}]
    constraintWeights [1,1,1]
   }
   "Transform (Modify) Bone - Bone: dyn_clavicle_cape_l"!(ComponentPose, Alpha=1, Translation="0,0,2", Rotation="0,0,0") -> (Pose) {
    rotationMode=BMM_Additive
    rotationSpace=BCS_BoneSpace
    translationMode=BMM_Additive
    translationSpace=BCS_BoneSpace
    alphaBoolBlend {blendOption=Linear}
    boneToModify {boneName=dyn_clavicle_cape_l}
    translation {z=2}
   }
   "Transform (Modify) Bone - Bone: dyn_clavicle_cape_r"!(ComponentPose, Alpha=1, Translation="0,3,5", Rotation="0,0,0") -> (Pose) {
    rotationMode=BMM_Additive
    rotationSpace=BCS_BoneSpace
    translationMode=BMM_Additive
    translationSpace=BCS_BoneSpace
    alphaBoolBlend {blendOption=Linear}
    boneToModify {boneName=dyn_clavicle_cape_r}
    translation {y=3, z=5}
   }
   n2 = Use cached pose 'trailCache'()
   n3 = Use cached pose 'dbnoCache'()
   n4 = Get bIsDBNO()
   Output Pose(Result=Blend Poses by bool(
     BlendPose_0=Use cached pose 'rigidCache'()
     BlendPose_1=Blend Poses by bool(BlendPose_0=@n3, BlendPose_1=@n2, bActiveValue=@n4) {blendType=Linear}
     bActiveValue=Get bRigidBodyEnabled()
    ) {blendType=Linear}) {layerGroup=None}
   n5 = RigidBody!(
    ComponentPose=Local To Component(LocalPose=Use cached pose 'Start_Cache'())
    Alpha
    OverrideWorldGravity:Vector=Get GravityOverride()
    ExternalForce
   ) -> (Pose) {
    bClampLinearTranslationLimitToRefPose=true
    bOverrideWorldGravity=true
    overridePhysicsAsset=&/F_MED_Obsidian_Original_Dyn_cape
    simulationSpace=BaseBoneSpace
    alphaBoolBlend {blendOption=Linear}
    baseBoneRef {boneName=dyn_root_simspace}
    componentAppliedLinearAccClamp=<3000 3000 3000>
    componentLinearAccScale=<0.1 0.1 0.5>
    componentLinearVelScale=<0.1 0.1 0.5>
    overrideWorldGravity {z=-980}
    simSpaceSettings {
     maxAngularAcceleration=1000
     maxAngularVelocity=3
     maxLinearAcceleration=700
     maxLinearVelocity=500
     velocityScaleZ=0
     worldAlpha=1
     externalLinearDragV=<1 1>
    }
   }
   n6 = "Transform (Modify) Bone - Bone: dyn_cape_pin_r_01"!(ComponentPose=@n5, Alpha=1, Translation="0,3,5") -> (Pose) {
    rotationMode=BMM_Additive
    rotationSpace=BCS_BoneSpace
    translationMode=BMM_Additive
    translationSpace=BCS_BoneSpace
    alphaBoolBlend {blendOption=Linear}
    alphaScaleBiasClamp {bInterpResult=true, interpSpeedDecreasing=5, interpSpeedIncreasing=7}
    boneToModify {boneName=dyn_cape_pin_r_01}
    translation {y=3, z=5}
   }
   n7 = "Transform (Modify) Bone - Bone: dyn_cape_mid_r_01"!(ComponentPose=@n6, Alpha=1, Translation="0,3,5") -> (Pose) {
    rotationMode=BMM_Additive
    rotationSpace=BCS_BoneSpace
    translationMode=BMM_Additive
    translationSpace=BCS_BoneSpace
    alphaBoolBlend {blendOption=Linear}
    alphaScaleBiasClamp {bInterpResult=true, interpSpeedDecreasing=5, interpSpeedIncreasing=7}
    boneToModify {boneName=dyn_cape_mid_r_01}
    translation {y=3, z=5}
   }
   n8 = RigidBody(ComponentPose=Local To Component(LocalPose=Use cached pose 'Start_Cache'()), OverrideWorldGravity=Get GravityOverride()) {
    bClampLinearTranslationLimitToRefPose=true
    bOverrideWorldGravity=true
    overridePhysicsAsset=&F_MED_Obsidian_Original_Dyn_cape
    simulationSpace=BaseBoneSpace
    alphaBoolBlend {blendOption=Linear}
    baseBoneRef {boneName=dyn_root_simspace}
    componentAppliedLinearAccClamp=<3000 3000 3000>
    componentLinearAccScale {z=0.5}
    componentLinearVelScale {z=0.5}
    overrideWorldGravity {z=-980}
    simSpaceSettings {
     maxAngularAcceleration=1000
     maxAngularVelocity=3
     maxLinearAcceleration=700
     maxLinearVelocity=500
     velocityScaleZ=0
     worldAlpha=1
     externalLinearDragV=<1 1>
    }
   }
   n9 = "Constraint - dyn_cape_l_02 (P)"(ComponentPose=@n8, Alpha=1, ConstraintWeights_0=1) {
    alphaBoolBlend {blendOption=Linear}
    boneToModify {boneName=dyn_cape_l_02}
    constraintSetup [{targetBone {boneName=spine_04}}]
    constraintWeights [1]
   }
   n10 = "Constraint - dyn_cape_r_02 (P)"(ComponentPose=@n9, Alpha=1, ConstraintWeights_0=1) {
    alphaBoolBlend {blendOption=Linear}
    boneToModify {boneName=dyn_cape_r_02}
    constraintSetup [{targetBone {boneName=spine_04}}]
    constraintWeights [1]
   }
   n11 = "Constraint - dyn_cape_mid_l_02 (P)"(ComponentPose=@n10, Alpha=1, ConstraintWeights_0=1) {
    alphaBoolBlend {blendOption=Linear}
    boneToModify {boneName=dyn_cape_mid_l_02}
    constraintSetup [{targetBone {boneName=spine_04}}]
    constraintWeights [1]
   }
   n12 = "Constraint - dyn_cape_mid_r_02 (P)"(ComponentPose=@n11, Alpha=1, ConstraintWeights_0=1) {
    alphaBoolBlend {blendOption=Linear}
    boneToModify {boneName=dyn_cape_mid_r_02}
    constraintSetup [{targetBone {boneName=spine_04}}]
    constraintWeights [1]
   }
   Save cached pose 'rigidCache'(Pose=Blend Poses by bool(BlendPose_0=Component To Local(ComponentPose=@n12), BlendPose_1=Component To Local(ComponentPose=@n7), bActiveValue=Get isWearingBackbling()) {blendType=Linear}) {cacheName=rigidCache}
   n13 = Get TrailControllerRot() -> (TrailControllerRot:Rotator)
   n14 = "Transform (Modify) Bone - Bone: dyn_cape_l_02"(ComponentPose=Local To Component(LocalPose=Use cached pose 'Start_Cache'()), Alpha=1, Translation="0,0,0", Rotation=@n13, Scale="1,1,1") {rotationMode=BMM_Additive, rotationSpace=BCS_BoneSpace, alphaBoolBlend {blendOption=Linear}, boneToModify {boneName=dyn_cape_l_02}}
   n15 = "Transform (Modify) Bone - Bone: dyn_cape_r_01"(ComponentPose=@n14, Alpha=1, Translation="0,0,0", Rotation=@n13, Scale="1,1,1") {rotationMode=BMM_Additive, rotationSpace=BCS_BoneSpace, alphaBoolBlend {blendOption=Linear}, boneToModify {boneName=dyn_cape_r_01}}
   n16 = "Transform (Modify) Bone - Bone: dyn_cape_mid_l_01"(ComponentPose=@n15, Alpha=1, Translation="0,0,0", Rotation=@n13, Scale="1,1,1") {rotationMode=BMM_Additive, rotationSpace=BCS_BoneSpace, alphaBoolBlend {blendOption=Linear}, boneToModify {boneName=dyn_cape_mid_l_01}}
   n17 = "Transform (Modify) Bone - Bone: dyn_cape_mid_r_01"!(ComponentPose=@n16, Alpha=1, Translation="0,0,0", Rotation:Rotator=@n13, Scale:Vector="1,1,1") -> (Pose) {rotationMode=BMM_Additive, rotationSpace=BCS_BoneSpace, alphaBoolBlend {blendOption=Linear}, boneToModify {boneName=dyn_cape_mid_r_01}}
   n18 = Get BwdMovement() -> (BwdMovement:double)
   n19 = Trail controller - Bone: dyn_cape_l_06!(ComponentPose=@n17, Alpha=@n18, RelaxationSpeedScale:float=1, FakeVelocity:Vector="0,0,0") -> (Pose) {
    bUsePlanarLimit=true
    chainLength=6
    debugLifeTime=0.1
    maxDeltaTime=0.0333
    relaxationSpeedScale=1.2
    alphaScaleBiasClamp {bInterpResult=true}
    rotationLimits [
     {limitMax=<180 90 180>, limitMin=<-180 -90 -180>}
     {limitMax=<180 90 180>, limitMin=<-180 -90 -180>}
     {limitMax=<180 90 180>, limitMin=<-180 -90 -180>}
     {limitMax=<180 90 180>, limitMin=<-180 -90 -180>}
     {limitMax=<180 90 180>, limitMin=<-180 -90 -180>}
     {limitMax=<180 180 180>}
    ]
    rotationOffsets [<0 0 0>, <0 0 0>, <0 0 0>, <0 0 0>, <0 0 0>, <0 0 0>]
    trailBone {boneName=dyn_cape_l_06}
    trailRelaxationSpeed {editorCurveData {keys [<at 0 10 RCIM_Linear>,<at 1 6 RCIM_Linear>]}}
   }
   n20 = Trail controller - Bone: dyn_cape_mid_l_06!(ComponentPose=@n19, Alpha=@n18, RelaxationSpeedScale:float=1, FakeVelocity:Vector="0,0,0") -> (Pose) {
    bUsePlanarLimit=true
    chainLength=6
    debugLifeTime=0.1
    maxDeltaTime=0.0333
    relaxationSpeedScale=1.2
    alphaScaleBiasClamp {bInterpResult=true}
    rotationLimits [
     {limitMax=<180 180 180>}
     {limitMax=<180 180 180>}
     {limitMax=<180 90 180>, limitMin=<-180 -90 -180>}
     {limitMax=<180 180 180>}
     {limitMax=<180 180 180>}
     {limitMax=<180 180 180>}
    ]
    rotationOffsets [<0 0 0>, <0 0 0>, <0 0 0>, <0 0 0>, <0 0 0>, <0 0 0>]
    trailBone {boneName=dyn_cape_mid_l_06}
    trailRelaxationSpeed {editorCurveData {keys [<at 0 10 RCIM_Linear>,<at 1 6 RCIM_Linear>]}}
   }
   n21 = Trail controller - Bone: dyn_cape_mid_r_06!(ComponentPose=@n20, Alpha=@n18, RelaxationSpeedScale:float=1, FakeVelocity:Vector="0,0,0") -> (Pose) {
    bUsePlanarLimit=true
    chainLength=6
    debugLifeTime=0.1
    maxDeltaTime=0.0333
    relaxationSpeedScale=1.2
    alphaScaleBiasClamp {bInterpResult=true}
    rotationLimits [
     {limitMax=<180 180 180>}
     {limitMax=<180 180 180>}
     {limitMax=<180 90 180>, limitMin=<-180 -90 -180>}
     {limitMax=<180 180 180>}
     {limitMax=<180 180 180>}
     {limitMax=<180 180 180>}
    ]
    rotationOffsets [<0 0 0>, <0 0 0>, <0 0 0>, <0 0 0>, <0 0 0>, <0 0 0>]
    trailBone {boneName=dyn_cape_mid_r_06}
    trailRelaxationSpeed {editorCurveData {keys [<at 0 10 RCIM_Linear>,<at 1 6 RCIM_Linear>]}}
   }
   n22 = Trail controller - Bone: dyn_cape_r_06!(ComponentPose=@n21, Alpha=@n18, RelaxationSpeedScale:float=1, FakeVelocity:Vector="0,0,0") -> (Pose) {
    bUsePlanarLimit=true
    chainLength=6
    debugLifeTime=0.1
    maxDeltaTime=0.0333
    relaxationSpeedScale=1.2
    alphaScaleBiasClamp {bInterpResult=true}
    rotationLimits [
     {limitMax=<180 90 180>, limitMin=<-180 -90 -180>}
     {limitMax=<180 90 180>, limitMin=<-180 -90 -180>}
     {limitMax=<180 90 180>, limitMin=<-180 -90 -180>}
     {limitMax=<180 90 180>, limitMin=<-180 -90 -180>}
     {limitMax=<180 90 180>, limitMin=<-180 -90 -180>}
     {limitMax=<180 180 180>}
    ]
    rotationOffsets [<0 0 0>, <0 0 0>, <0 0 0>, <0 0 0>, <0 0 0>, <0 0 0>]
    trailBone {boneName=dyn_cape_r_06}
    trailRelaxationSpeed {editorCurveData {keys [<at 0 10 RCIM_Linear>,<at 1 6 RCIM_Linear>]}}
   }
   n23 = "Transform (Modify) Bone - Bone: dyn_cape_l_01"!(ComponentPose=@n22, Alpha=1, Translation="0,0,2", Rotation="0,0,0") -> (Pose) {
    rotationMode=BMM_Additive
    rotationSpace=BCS_BoneSpace
    translationMode=BMM_Additive
    translationSpace=BCS_BoneSpace
    alphaBoolBlend {blendOption=Linear}
    boneToModify {boneName=dyn_cape_l_01}
    translation {z=2}
   }
   n24 = "Transform (Modify) Bone - Bone: dyn_cape_pin_mid_l_01"!(ComponentPose=@n23, Alpha=1, Translation="0,0,2", Rotation="0,0,0") -> (Pose) {
    rotationMode=BMM_Additive
    rotationSpace=BCS_BoneSpace
    translationMode=BMM_Additive
    translationSpace=BCS_BoneSpace
    alphaBoolBlend {blendOption=Linear}
    boneToModify {boneName=dyn_cape_pin_mid_l_01}
    translation {z=2}
   }
   n25 = "Transform (Modify) Bone - Bone: dyn_cape_pin_mid_r_01"!(ComponentPose=@n24, Alpha=1, Translation="0,0,2", Rotation="0,0,0") -> (Pose) {
    rotationMode=BMM_Additive
    rotationSpace=BCS_BoneSpace
    translationMode=BMM_Additive
    translationSpace=BCS_BoneSpace
    alphaBoolBlend {blendOption=Linear}
    boneToModify {boneName=dyn_cape_pin_mid_r_01}
    translation {z=2}
   }
   Save cached pose 'trailCache'(Pose=Component To Local(ComponentPose=@n25)) {cacheName=trailCache}
   n26 = "Transform (Modify) Bone - Bone: dyn_cape_l_01"!(
    ComponentPose=Local To Component(LocalPose=Use cached pose 'Start_Cache'())
    Alpha=1
    Translation="0,0,0"
    Rotation="5,0,0"
    Scale:Vector="1,1,1"
   ) -> (Pose) {
    rotationMode=BMM_Additive
    rotationSpace=BCS_BoneSpace
    alphaBoolBlend {blendOption=Linear}
    boneToModify {boneName=dyn_cape_l_01}
    rotation {pitch=5}
   }
   n27 = "Transform (Modify) Bone - Bone: dyn_cape_l_02"(ComponentPose=@n26, Alpha=1, Translation="0,0,0", Rotation="-10,0,0", Scale="1,1,1") {
    rotationMode=BMM_Additive
    rotationSpace=BCS_BoneSpace
    alphaBoolBlend {blendOption=Linear}
    boneToModify {boneName=dyn_cape_l_02}
    rotation {pitch=-10}
   }
   n28 = "Transform (Modify) Bone - Bone: dyn_cape_l_05"(ComponentPose=@n27, Alpha=1, Translation="0,0,0", Rotation="-25,0,0", Scale="1,1,1") {
    rotationMode=BMM_Additive
    rotationSpace=BCS_BoneSpace
    alphaBoolBlend {blendOption=Linear}
    boneToModify {boneName=dyn_cape_l_05}
    rotation {pitch=-25}
   }
   n29 = "Transform (Modify) Bone - Bone: dyn_cape_r_01"(ComponentPose=@n28, Alpha=1, Translation="0,0,0", Rotation="5,0,0", Scale="1,1,1") {
    rotationMode=BMM_Additive
    rotationSpace=BCS_BoneSpace
    alphaBoolBlend {blendOption=Linear}
    boneToModify {boneName=dyn_cape_r_01}
    rotation {pitch=5}
   }
   n30 = "Transform (Modify) Bone - Bone: dyn_cape_r_02"(ComponentPose=@n29, Alpha=1, Translation="0,0,0", Rotation="-10,0,0", Scale="1,1,1") {
    rotationMode=BMM_Additive
    rotationSpace=BCS_BoneSpace
    alphaBoolBlend {blendOption=Linear}
    boneToModify {boneName=dyn_cape_r_02}
    rotation {pitch=-10}
   }
   n31 = "Transform (Modify) Bone - Bone: dyn_cape_r_05"(ComponentPose=@n30, Alpha=1, Translation="0,0,0", Rotation="-25,0,0", Scale="1,1,1") {
    rotationMode=BMM_Additive
    rotationSpace=BCS_BoneSpace
    alphaBoolBlend {blendOption=Linear}
    boneToModify {boneName=dyn_cape_r_05}
    rotation {pitch=-25}
   }
   n32 = "Transform (Modify) Bone - Bone: dyn_cape_mid_l_01"(ComponentPose=@n31, Alpha=1, Translation="0,0,0", Rotation="5,0,0", Scale="1,1,1") {
    rotationMode=BMM_Additive
    rotationSpace=BCS_BoneSpace
    alphaBoolBlend {blendOption=Linear}
    boneToModify {boneName=dyn_cape_mid_l_01}
    rotation {pitch=5}
   }
   n33 = "Transform (Modify) Bone - Bone: dyn_cape_mid_l_02"(ComponentPose=@n32, Alpha=1, Translation="0,0,0", Rotation="-10,0,0", Scale="1,1,1") {
    rotationMode=BMM_Additive
    rotationSpace=BCS_BoneSpace
    alphaBoolBlend {blendOption=Linear}
    boneToModify {boneName=dyn_cape_mid_l_02}
    rotation {pitch=-10}
   }
   n34 = "Transform (Modify) Bone - Bone: dyn_cape_mid_l_05"(ComponentPose=@n33, Alpha=1, Translation="0,0,0", Rotation="-25,0,0", Scale="1,1,1") {
    rotationMode=BMM_Additive
    rotationSpace=BCS_BoneSpace
    alphaBoolBlend {blendOption=Linear}
    boneToModify {boneName=dyn_cape_mid_l_05}
    rotation {pitch=-25}
   }
   n35 = "Transform (Modify) Bone - Bone: dyn_cape_r_01"(ComponentPose=@n34, Alpha=1, Translation="0,0,0", Rotation="5,0,0", Scale="1,1,1") {
    rotationMode=BMM_Additive
    rotationSpace=BCS_BoneSpace
    alphaBoolBlend {blendOption=Linear}
    boneToModify {boneName=dyn_cape_r_01}
    rotation {pitch=5}
   }
   n36 = "Transform (Modify) Bone - Bone: dyn_cape_mid_r_02"(ComponentPose=@n35, Alpha=1, Translation="0,0,0", Rotation="-10,0,0", Scale="1,1,1") {
    rotationMode=BMM_Additive
    rotationSpace=BCS_BoneSpace
    alphaBoolBlend {blendOption=Linear}
    boneToModify {boneName=dyn_cape_mid_r_02}
    rotation {pitch=-10}
   }
   n37 = "Transform (Modify) Bone - Bone: dyn_cape_mid_r_05"(ComponentPose=@n36, Alpha=1, Translation="0,0,0", Rotation="-25,0,0", Scale="1,1,1") {
    rotationMode=BMM_Additive
    rotationSpace=BCS_BoneSpace
    alphaBoolBlend {blendOption=Linear}
    boneToModify {boneName=dyn_cape_mid_r_05}
    rotation {pitch=-25}
   }
   n38 = Trail controller - Bone: dyn_cape_mid_l_06(ComponentPose=@n37, Alpha=1, RelaxationSpeedScale=2, FakeVelocity="0,0,0") {
    bActorSpaceFakeVel=true
    bShowTrailLocation=true
    bUsePlanarLimit=true
    chainLength=5
    debugLifeTime=0.1
    maxDeltaTime=0.0333
    relaxationSpeedScale=1.2
    alphaScaleBiasClamp {bInterpResult=true, interpSpeedDecreasing=25, interpSpeedIncreasing=25}
    baseJoint {boneName=root}
    rotationLimits [{limitMax=<180 180 180>}, {limitMax=<180 180 180>}, {limitMax=<180 180 180>}, {limitMax=<180 180 180>}, {limitMax=<180 180 180>}]
    rotationOffsets [<0 0 0>, <0 0 0>, <0 0 0>, <0 0 0>, <0 0 0>]
    trailBone {boneName=dyn_cape_mid_l_06}
    trailRelaxationSpeed {editorCurveData {keys [<at 0 15 RCIM_Linear>,<at 1 6 RCIM_Linear>]}}
   }
   n39 = Trail controller - Bone: dyn_cape_mid_r_06(ComponentPose=@n38, Alpha=1, RelaxationSpeedScale=2, FakeVelocity="0,0,0") {
    bActorSpaceFakeVel=true
    bShowTrailLocation=true
    bUsePlanarLimit=true
    chainLength=5
    debugLifeTime=0.1
    maxDeltaTime=0.0333
    relaxationSpeedScale=1.2
    alphaScaleBiasClamp {bInterpResult=true, interpSpeedDecreasing=25, interpSpeedIncreasing=25}
    baseJoint {boneName=root}
    rotationLimits [{limitMax=<180 180 180>}, {limitMax=<180 180 180>}, {limitMax=<180 180 180>}, {limitMax=<180 180 180>}, {limitMax=<180 180 180>}]
    rotationOffsets [<0 0 0>, <0 0 0>, <0 0 0>, <0 0 0>, <0 0 0>]
    trailBone {boneName=dyn_cape_mid_r_06}
    trailRelaxationSpeed {editorCurveData {keys [<at 0 15 RCIM_Linear>,<at 1 6 RCIM_Linear>]}}
   }
   n40 = Trail controller - Bone: dyn_cape_l_06(ComponentPose=@n39, Alpha=1, RelaxationSpeedScale=2, FakeVelocity="0,0,0") {
    bActorSpaceFakeVel=true
    bShowTrailLocation=true
    bUsePlanarLimit=true
    chainLength=6
    debugLifeTime=0.1
    maxDeltaTime=0.0333
    relaxationSpeedScale=1.2
    alphaScaleBiasClamp {bInterpResult=true, interpSpeedDecreasing=25, interpSpeedIncreasing=25}
    baseJoint {boneName=root}
    rotationLimits [
     {limitMax=<180 180 180>}
     {limitMax=<180 180 180>}
     {limitMax=<180 180 180>}
     {limitMax=<180 180 180>}
     {limitMax=<180 180 180>}
     {limitMax=<180 180 180>}
    ]
    rotationOffsets [<0 0 0>, <0 0 0>, <0 0 0>, <0 0 0>, <0 0 0>, <0 0 0>]
    trailBone {boneName=dyn_cape_l_06}
    trailRelaxationSpeed {editorCurveData {keys [<at 0 15 RCIM_Linear>,<at 1 6 RCIM_Linear>]}}
   }
   n41 = Trail controller - Bone: dyn_cape_r_06(ComponentPose=@n40, Alpha=1, RelaxationSpeedScale=2, FakeVelocity="0,0,0") {
    bActorSpaceFakeVel=true
    bShowTrailLocation=true
    bUsePlanarLimit=true
    chainLength=6
    debugLifeTime=0.1
    maxDeltaTime=0.0333
    relaxationSpeedScale=1.2
    alphaScaleBiasClamp {bInterpResult=true, interpSpeedDecreasing=25, interpSpeedIncreasing=25}
    baseJoint {boneName=root}
    rotationLimits [
     {limitMax=<180 180 180>}
     {limitMax=<180 180 180>}
     {limitMax=<180 180 180>}
     {limitMax=<180 180 180>}
     {limitMax=<180 180 180>}
     {limitMax=<180 180 180>}
    ]
    rotationOffsets [<0 0 0>, <0 0 0>, <0 0 0>, <0 0 0>, <0 0 0>, <0 0 0>]
    trailBone {boneName=dyn_cape_r_06}
    trailRelaxationSpeed {editorCurveData {keys [<at 0 15 RCIM_Linear>,<at 1 6 RCIM_Linear>]}}
   }
   Save cached pose 'dbnoCache'(Pose=Component To Local(ComponentPose=@n41)) {cacheName=dbnoCache}
   n42 = Control Rig(Source=Copy Pose From Mesh() {bUseAttachedParent=true}) {
    alphaBoolBlend {blendOption=Linear}
    controlRigAssetReference {blueprintRigClass=&/F_MED_Obsidian_Original_Deform_CtrlRig_twists.F_MED_Obsidian_Original_Deform_CtrlRig_twists_C}
   }
   n43 = Control Rig!(
    Source=@n42
    Alpha
    KneeRot_fr_l_114deg:Vector
    KneeTrans_fr_l_114deg:Vector="0.258,-3.591,0.009"
    KneeScale_fr_l_114deg:Vector="1,1,1"
    KneeScale_fr_l_130deg:Vector="1,1,1"
    KneeRot_fr_l_130deg:Vector="0,0,-51.2"
    KneeTrans_fr_l_130deg:Vector="0.258,-3.591,0.009"
   ) -> (Pose) {
    alphaBoolBlend {blendOption=Linear}
    controlRigAssetReference {blueprintRigClass=&/F_MED_Obsidian_Original_Deform_CtrlRig_sculpters.F_MED_Obsidian_Original_Deform_CtrlRig_sculpters_C}
   }
   n44 = "Constraint - dyn_cape_disc_fr_l_MSTR (PPPP)"!(
    ComponentPose=Local To Component(LocalPose=@n43)
    Alpha=1
    ConstraintWeights_0=0.0421
    ConstraintWeights_1:float=0.529
    ConstraintWeights_2:float=0.3558
    ConstraintWeights_3:float=0.0731
   ) -> (Pose) {
    alphaBoolBlend {blendOption=Linear}
    boneToModify {boneName=dyn_cape_disc_fr_l_MSTR}
    constraintSetup [
     {targetBone {boneName=deform_pectoral_fr_l}}
     {targetBone {boneName=spine_04}}
     {targetBone {boneName=spine_05}}
     {targetBone {boneName=clavicle_l}}
    ]
    constraintWeights [1,1,1,1]
   }
   n45 = "Constraint - dyn_cape_disc_fr_r_MSTR (PPPP)"!(
    ComponentPose=@n44
    Alpha=1
    ConstraintWeights_0=0.0734
    ConstraintWeights_1:float=0.5297
    ConstraintWeights_2:float=0.3553
    ConstraintWeights_3:float=0.0417
   ) -> (Pose) {
    alphaBoolBlend {blendOption=Linear}
    boneToModify {boneName=dyn_cape_disc_fr_r_MSTR}
    constraintSetup [
     {targetBone {boneName=clavicle_r}}
     {targetBone {boneName=spine_04}}
     {targetBone {boneName=spine_05}}
     {targetBone {boneName=deform_pectoral_fr_r}}
    ]
    constraintWeights [1,1,1,1]
   }
   n46 = "Constraint - dyn_belt_discs_bk_MSTR (PPP)"!(ComponentPose=@n45, Alpha=1, ConstraintWeights_0=0.1836, ConstraintWeights_1:float=0.1323, ConstraintWeights_2:float=0.6841) -> (Pose) {
    alphaBoolBlend {blendOption=Linear}
    boneToModify {boneName=dyn_belt_discs_bk_MSTR}
    constraintSetup [{targetBone {boneName=spine_01}}, {targetBone {boneName=spine_02}}, {targetBone {boneName=pelvis}}]
    constraintWeights [1,1,1]
   }
   n47 = "Constraint - dyn_belt_discs_bk_l_MSTR (PPPP)"!(
    ComponentPose=@n46
    Alpha=1
    ConstraintWeights_0=0.183
    ConstraintWeights_1:float=0.2017
    ConstraintWeights_2:float=0.0136
    ConstraintWeights_3:float=0.6017
   ) -> (Pose) {
    alphaBoolBlend {blendOption=Linear}
    boneToModify {boneName=dyn_belt_discs_bk_l_MSTR}
    constraintSetup [{targetBone {boneName=spine_01}}, {targetBone {boneName=spine_02}}, {targetBone {boneName=spine_03}}, {targetBone {boneName=pelvis}}]
    constraintWeights [1,1,1,1]
   }
   n48 = "Constraint - dyn_belt_discs_l_MSTR (PPPPPP)"!(
    ComponentPose=@n47
    Alpha=1
    ConstraintWeights_0=0.0473
    ConstraintWeights_1:float=0.0963
    ConstraintWeights_2:float=0.1435
    ConstraintWeights_3:float=0.0384
    ConstraintWeights_4:float=0.0261
    ConstraintWeights_5:float=0.6484
   ) -> (Pose) {
    alphaBoolBlend {blendOption=Linear}
    boneToModify {boneName=dyn_belt_discs_l_MSTR}
    constraintSetup [
     {targetBone {boneName=deform_thigh_fr_l}}
     {targetBone {boneName=spine_01}}
     {targetBone {boneName=spine_02}}
     {targetBone {boneName=thigh_twist_01_l}}
     {targetBone {boneName=spine_04}}
     {targetBone {boneName=pelvis}}
    ]
    constraintWeights [1,1,1,1,1,1]
   }
   n49 = "Constraint - dyn_belt_discs_fr_l_MSTR (PPPPPP)"!(
    ComponentPose=@n48
    Alpha=1
    ConstraintWeights_0=0.1986
    ConstraintWeights_1:float=0.0257
    ConstraintWeights_2:float=0.0444
    ConstraintWeights_3:float=0.0248
    ConstraintWeights_4:float=0.0106
    ConstraintWeights_5:float=0.6959
   ) -> (Pose) {
    alphaBoolBlend {blendOption=Linear}
    boneToModify {boneName=dyn_belt_discs_fr_l_MSTR}
    constraintSetup [
     {targetBone {boneName=deform_thigh_fr_l}}
     {targetBone {boneName=spine_01}}
     {targetBone {boneName=spine_02}}
     {targetBone {boneName=thigh_twist_01_l}}
     {targetBone {boneName=spine_04}}
     {targetBone {boneName=pelvis}}
    ]
    constraintWeights [1,1,1,1,1,1]
   }
   n50 = "Constraint - dyn_belt_discs_fr_MSTR (PP)"!(ComponentPose=@n49, Alpha=1, ConstraintWeights_0=0.0831, ConstraintWeights_1:float=0.9169) -> (Pose) {
    alphaBoolBlend {blendOption=Linear}
    boneToModify {boneName=dyn_belt_discs_fr_MSTR}
    constraintSetup [{targetBone {boneName=deform_thigh_fr_l}}, {targetBone {boneName=pelvis}}]
    constraintWeights [1,1]
   }
   n51 = "Constraint - dyn_belt_discs_bk_r_MSTR (PPPP)"!(
    ComponentPose=@n50
    Alpha=1
    ConstraintWeights_0=0.184
    ConstraintWeights_1:float=0.2029
    ConstraintWeights_2:float=0.0137
    ConstraintWeights_3:float=0.5995
   ) -> (Pose) {
    alphaBoolBlend {blendOption=Linear}
    boneToModify {boneName=dyn_belt_discs_bk_r_MSTR}
    constraintSetup [{targetBone {boneName=spine_01}}, {targetBone {boneName=spine_02}}, {targetBone {boneName=spine_03}}, {targetBone {boneName=pelvis}}]
    constraintWeights [1,1,1,1]
   }
   n52 = "Constraint - dyn_belt_discs_r_MSTR (PPPPPP)"!(
    ComponentPose=@n51
    Alpha=1
    ConstraintWeights_0=0.0376
    ConstraintWeights_1:float=0.0978
    ConstraintWeights_2:float=0.1456
    ConstraintWeights_3:float=0.0265
    ConstraintWeights_4:float=0.6463
    ConstraintWeights_5:float=0.0463
   ) -> (Pose) {
    alphaBoolBlend {blendOption=Linear}
    boneToModify {boneName=dyn_belt_discs_r_MSTR}
    constraintSetup [
     {targetBone {boneName=thigh_twist_01_r}}
     {targetBone {boneName=spine_01}}
     {targetBone {boneName=spine_02}}
     {targetBone {boneName=spine_04}}
     {targetBone {boneName=pelvis}}
     {targetBone {boneName=deform_thigh_fr_r}}
    ]
    constraintWeights [1,1,1,1,1,1]
   }
   n53 = "Constraint - dyn_belt_discs_fr_r_MSTR (PPPPP)"!(
    ComponentPose=@n52
    Alpha=1
    ConstraintWeights_0=0.7549
    ConstraintWeights_1:float=0.0244
    ConstraintWeights_2:float=0.0488
    ConstraintWeights_3:float=0.0101
    ConstraintWeights_4:float=0.1619
   ) -> (Pose) {
    alphaBoolBlend {blendOption=Linear}
    boneToModify {boneName=dyn_belt_discs_fr_r_MSTR}
    constraintSetup [
     {targetBone {boneName=pelvis}}
     {targetBone {boneName=spine_01}}
     {targetBone {boneName=spine_02}}
     {targetBone {boneName=spine_04}}
     {targetBone {boneName=deform_thigh_fr_r}}
    ]
    constraintWeights [1,1,1,1,1]
   }
   n54 = "Constraint - dyn_cape_pin_bk_out_l_01_MSTR (R)"(ComponentPose=@n53, Alpha=1, ConstraintWeights_0=1) {
    alphaBoolBlend {blendOption=Linear}
    boneToModify {boneName=dyn_cape_pin_bk_out_l_01_MSTR}
    constraintSetup [{transformType=Rotation, targetBone {boneName=spine_05}}]
    constraintWeights [1]
   }
   n55 = "Constraint - dyn_cape_pin_bk_out_r_01_MSTR (R)"(ComponentPose=@n54, Alpha=1, ConstraintWeights_0=1) {
    alphaBoolBlend {blendOption=Linear}
    boneToModify {boneName=dyn_cape_pin_bk_out_r_01_MSTR}
    constraintSetup [{transformType=Rotation, targetBone {boneName=spine_05}}]
    constraintWeights [1]
   }
   n56 = "Transform (Modify) Bone - Bone: dyn_root_simspace"(ComponentPose=@n55, Alpha=1, Translation="0,0,0", Rotation=Get sim_space_rotation() -> (sim_space_rotation:Rotator), Scale="1,1,1") {
    rotationMode=BMM_Replace
    rotationSpace=BCS_WorldSpace
    translationSpace=BCS_BoneSpace
    alphaBoolBlend {blendOption=Linear}
    alphaScaleBiasClamp {bInterpResult=true, interpSpeedDecreasing=2, interpSpeedIncreasing=2}
    boneToModify {boneName=dyn_root_simspace}
   }
   n57 = "Transform (Modify) Bone - Bone: dyn_crouch_collision"(
    ComponentPose=@n56
    Alpha=Get crouch alpha inv() -> (crouch alpha inv:double)
    Translation="0.000,0.000,-49.249"
    Rotation="0,0,0"
    Scale="1,1,1"
   ) {
    translationMode=BMM_Replace
    translationSpace=BCS_BoneSpace
    alphaBoolBlend {blendOption=Linear}
    alphaScaleBiasClamp {bInterpResult=true}
    boneToModify {boneName=dyn_crouch_collision}
    translation {z=-49.249}
   }
   n58 = "Transform (Modify) Bone - Bone: dyn_root_slope"!(
    ComponentPose=@n57
    bAlphaBoolEnabled:bool=NOT Boolean(A=Get bIsInFrontEnd())
    Translation=Get slope_root_trans() -> (slope_root_trans:Vector)
    Rotation=Get slope_foot_rot() -> (slope_foot_rot:Rotator)
    Scale="1,1,1"
   ) -> (Pose) {
    alphaInputType=Bool
    rotationMode=BMM_Replace
    rotationSpace=BCS_WorldSpace
    translationMode=BMM_Replace
    translationSpace=BCS_WorldSpace
    alphaBoolBlend {blendOption=Linear}
    boneToModify {boneName=dyn_root_slope}
   }
   Save cached pose 'Start_Cache'(Pose=Component To Local(ComponentPose=@n58)) {cacheName=Start_Cache}
  ]}
 EventGraph {nodes [
   DrawVectorFromJointFunction!(
    execute
    SkelMeshComponent:Scene Component Object Reference
    JointName:Name=root
    LineEnd:Vector="0, 0, 0"
    JointOffset:Vector="0, 0, 0"
    LineColor:Linear Color Structure=(R=1,G=0,B=0,A=1)
    Thickness:double=2
    Foreground:EDrawDebugSceneDepthPriorityGroup Enum=Foreground
   ) -> (then)
   > DrawJointDebugFunction!(
    execute
    SkelMeshComponent:Scene Component Object Reference
    JointOffset:Vector="0, 0, 0"
    JointName:Array of Names=Make Array("[0]":Name=root) -> (Array:Array of Names)
    Scale:double=10
    Duration:double=0
    Thickness:double=0.7
    DebugDraw:bool=true
    Foreground:EDrawDebugSceneDepthPriorityGroup Enum=Foreground
   ) -> (then)
   n1 = Event BlueprintUpdateAnimation()
   > n2 = Sequence!(execute) -> (then_0, then_1, then_2:Exec, then_3:Exec, then_4:Exec)
   > n3 = Gravity Override State Function DEPRECATED(CharacterPartAnimInstance=Self-Reference(), GameTimeInSeconds=GetGameTimeInSeconds(), GravityOverrideParams=Get Gravity Override Params())
   > Set GravityOverride(GravityOverride=@n3)
   n4 = ComplexSimSpaceFunction DEPRECATED!(
    execute=@n2.then_1
    CharacterPartAnimInstance:Custom Character Part Anim Instance Object Reference=Self-Reference()
    InterpSpeedMax:double=9
    PawnSpeedMax:double=20
    RInterpErrorTolerance:double=1
    DeltaTime:double=@n1.DeltaTimeX
    DebugDraw:bool=false
    sim_space_rotation_in:Rotator=Get sim_space_rotation()
    inComponentSpaceIn:bool=Get InComponentSpace() -> (InComponentSpace:bool)
    inComponentSpaceLastFrameIn:bool=Get InComponentSpaceLastFrame() -> (InComponentSpaceLastFrame:bool)
    lockedOnSimSpaceIn:bool=Get LockedOnSimSpace() -> (LockedOnSimSpace:bool)
    InterpSpeedIn:double=Get InterpSpeed() -> (InterpSpeed:double)
   ) -> (
    then
    sim_space_rotation:Rotator
    inComponentSpaceOut:bool
    inComponentSpaceLastFrameOut:bool
    lockedOnSimSpaceOut:bool
    InterpSpeedOut:double
   )
   > Set sim_space_rotation(sim_space_rotation:Rotator=@n4) -> (Output_Get:Rotator)
   > Set InComponentSpace(InComponentSpace:bool=@n4.inComponentSpaceOut) -> (Output_Get:bool)
   > Set InComponentSpaceLastFrame(InComponentSpaceLastFrame:bool=@n4.inComponentSpaceLastFrameOut) -> (Output_Get:bool)
   > Set LockedOnSimSpace(LockedOnSimSpace:bool=@n4.lockedOnSimSpaceOut) -> (Output_Get:bool)
   > Set InterpSpeed(InterpSpeed:double=@n4.InterpSpeedOut) -> (Output_Get:double)
   n5 = Break Fort Anim Input Slope Info!(FortAnimInput_SlopeInfo:"Fort Anim Input Slope Info Structure (by ref)"=Get SlopeInfoInput() -> (SlopeInfoInput:Fort Anim Input Slope Info Structure)) -> (RootSlopeTranslation:Vector=~, LeftFootSlopeTranslation:Vector=~, RightFootSlopeTranslation:Vector=~, SlopeRotation:Rotator=~) {is=title K2Node_BreakStruct}
   Set slope_root_trans(execute=@n2.then_2, slope_root_trans:Vector=@n5) -> (Output_Get:Vector)
   > Set slope_foot_rot(slope_foot_rot:Rotator=@n5.SlopeRotation) -> (Output_Get:Rotator)
   n6 = PlayerMovement(execute=@n2.then_3, AnimInstance=Self-Reference(), XDotProductCoefficient=1, YDotProductCoefficient=1, ZDotProductCoefficient=1)
   n7 = MapRangeClamped(Value=@n6.YMovement, InRangeA=0, InRangeB=-1)
   n8 = "Absolute (Float)"(A=@n6)
   > Set BwdMovement(BwdMovement:double=MapRangeClamped(Value=float + float(A=@n8, B=@n7) {is=title K2Node_CommutativeAssociativeBinaryOperator}, InRangeA=0, InRangeB=1.4, OutRangeA=1, OutRangeB=0)) -> (Output_Get:double)
   n9 = Branch(execute=@n2.then_4, Condition=Get bIsSwimSprinting() -> (bIsSwimSprinting:bool))
   > Set TrailControllerRot(TrailControllerRot:Rotator="15,0, 0") -> (Output_Get:Rotator)
   n10 = Branch(execute=@n9.else, Condition=Get bIsSkyDiving())
   > n11 = Branch(Condition=Get bIsParachuteOpened())
   > Set TrailControllerRot(TrailControllerRot="45,0, 0")
   n12 = Branch(execute=@n11.else, Condition="float > float"(A=Get PawnVelocityZ(), B=50) {is=title K2Node_CallFunction})
   > Set BwdMovement(BwdMovement=float * float(A=Get BwdMovement(), B=0))
   > Set TrailControllerRot(TrailControllerRot="45,0, 0")
   n13 = Branch(execute=@n12.else, Condition=Get bIsSkydiveDiveMode())
   > Set TrailControllerRot(TrailControllerRot="7,0, 0")
   Set TrailControllerRot(execute=@n13.else, TrailControllerRot="60,0, 0")
   Set TrailControllerRot(execute=@n10.else, TrailControllerRot="15,0, 0")
   > Set BwdMovement(BwdMovement=float * float(A=Get BwdMovement(), B=1))
   Event BlueprintInitializeAnimation()
   > n14 = Set SkelMesh(SkelMesh=GetPartSkeletalMeshComponent())
   Event OnEndCrouching()
   > Set crouch alpha(crouch alpha:double=0) -> (Output_Get:double)
   > Set crouch alpha inv(crouch alpha inv:double=1) -> (Output_Get:double)
   Event OnBeginCrouching()
   > Set crouch alpha(crouch alpha=1)
   > Set crouch alpha inv(crouch alpha inv=0)
   Event OnCharacterCustomizationLoaded()
   n15 = CheckForBackbling()
   > Set isWearingBackbling(isWearingBackbling=OR Boolean(A=@n15, B=@n15.bIsBackblingLong))
   EnableGravityOverride!(
    execute
    self:Clothing Simulation Interactor Object Reference=GetClothingSimulationInteractor!(self:Skeletal Mesh Component Object Reference=@n14, ClothingAssetName:Name=None) -> (ReturnValue:Clothing Simulation Interactor Object Reference)
    InVector:"Vector (by ref)"="0, 0, 0"
   ) -> (then)
  ]}
]
variables [
 BwdMovement {type=number}
 crouch alpha {type=number}
 crouch alpha inv {type=number, default=1}
 Gravity Override Params {type=GravityOverrideParamsStruct}
 GravityOverride {type=Vector}
 InComponentSpace {type=boolean}
 InComponentSpaceLastFrame {type=boolean}
 InterpSpeed {type=number}
 isWearingBackbling {type=boolean}
 LockedOnSimSpace {type=boolean}
 sim_space_rotation {type=Rotator}
 SkelMesh {type=SkeletalMeshComponent}
 slope_foot_rot {type=Rotator}
 slope_root_trans {type=Vector}
 TrailControllerRot {type=Rotator}
]
